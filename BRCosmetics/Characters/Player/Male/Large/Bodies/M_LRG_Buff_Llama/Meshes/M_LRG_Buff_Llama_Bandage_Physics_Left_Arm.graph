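# Physics asset for the left-arm bandage on the large male Buff Llama cosmetic body, defining capsule bodies for secondary motion of bandage and flap bones.
# A three-segment bandage chain constrains to the left master bandage bone, with a much wider swing on the middle joint.
# Three small flaps constrain tightly to the left lower-arm twist bone, which carries a large forearm capsule.
# Right-side master and twist bodies are tiny and unconstrained.
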PhysicsAsset /BRCosmetics/Characters/Player/Male/Large/Bodies/M_LRG_Buff_Llama/Meshes/M_LRG_Buff_Llama_Bandage_Physics_Left_Arm
bodies {
 Dyn_Bandage_l_01 [{shapeType=Capsule, length=2, radius=2, shapeName=Dyn_Bandage_l_01_capsule, center=<1 0 0>, rotation=<rot 90 0 0>}]
 Dyn_Bandage_l_02 [{shapeType=Capsule, length=2, radius=2, shapeName=Dyn_Bandage_l_02_capsule, center=<1 0 0>, rotation=<rot 90 0 0>}]
 Dyn_Bandage_l_03 [{shapeType=Capsule, length=2, radius=2, shapeName=Dyn_Bandage_l_03_capsule, center=<1 0 0>, rotation=<rot 90 0 0>}]
 Dyn_Flap_l1 [{shapeType=Capsule, length=1, radius=1, shapeName=Dyn_Flap_l1_capsule, center=<0.5 0 0>, rotation=<rot 90 0 0>}]
 Dyn_Flap_l2 [{shapeType=Capsule, length=1, radius=1, shapeName=Dyn_Flap_l2_capsule, center=<0.5 0 0>, rotation=<rot 90 0 0>}]
 Dyn_Flap_l3 [{shapeType=Capsule, length=1, radius=1, shapeName=Dyn_Flap_l3_capsule, center=<0.5 0 0>, rotation=<rot 90 0 0>}]
 Dyn_Master_Bandage_l [{shapeType=Capsule, length=0.5, radius=0.5, shapeName=Dyn_Master_Bandage_l_capsule}]
 Dyn_Master_Bandage_r [{shapeType=Capsule, length=0.5, radius=0.5, shapeName=Dyn_Master_Bandage_r_capsule}]
 lowerarm_twist_02_l [{shapeType=Capsule, length=10, radius=4.5, shapeName=lowerarm_twist_02_l_capsule, center=<4.5089 0 -0.0001>, rotation=<rot 90 0 0>}]
 lowerarm_twist_02_r [{shapeType=Capsule, length=0.505, radius=0.505, shapeName=lowerarm_twist_02_r_capsule}]
}
constraints [
 {bone1Name=Dyn_Bandage_l_03, bone2Name=Dyn_Bandage_l_02, swing1LimitDegrees=45, swing2LimitDegrees=45, twistLimitDegrees=45}
 {bone1Name=Dyn_Bandage_l_02, bone2Name=Dyn_Bandage_l_01, swing1LimitDegrees=160, swing2LimitDegrees=45, twistLimitDegrees=45}
 {bone1Name=Dyn_Bandage_l_01, bone2Name=Dyn_Master_Bandage_l, swing1LimitDegrees=45, swing2LimitDegrees=45, twistLimitDegrees=45}
 {bone1Name=Dyn_Flap_l1, bone2Name=lowerarm_twist_02_l, swing1LimitDegrees=20, swing2LimitDegrees=5, twistLimitDegrees=1}
 {bone1Name=Dyn_Flap_l2, bone2Name=lowerarm_twist_02_l, swing1LimitDegrees=20, swing2LimitDegrees=5, twistLimitDegrees=1}
 {bone1Name=Dyn_Flap_l3, bone2Name=lowerarm_twist_02_l, swing1LimitDegrees=20, swing2LimitDegrees=5, twistLimitDegrees=1}
]
