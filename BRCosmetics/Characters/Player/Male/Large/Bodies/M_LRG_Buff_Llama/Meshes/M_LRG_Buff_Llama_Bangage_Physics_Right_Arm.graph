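# Physics asset for the right arm of the large male Buff Llama player body, covering its dangling bandage and two arm flaps.
# A three-link bandage chain hangs from a master bandage bone, which rotates freely on all angular axes relative to lowerarm_twist_01_r.
# Both flaps attach separately to lowerarm_twist_02_r with much tighter angular limits.
# All bodies are capsules, with a larger one on lowerarm_twist_01_r approximating the forearm.
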PhysicsAsset /BRCosmetics/Characters/Player/Male/Large/Bodies/M_LRG_Buff_Llama/Meshes/M_LRG_Buff_Llama_Bangage_Physics_Right_Arm
bodies {
 Dyn_ArmFlap_1r [{shapeType=Capsule, length=1, radius=1, shapeName=Dyn_ArmFlap_1r_capsule, center=<0.5 0 0>, rotation=<rot 90 0 0>}]
 Dyn_ArmFlap_2r [{shapeType=Capsule, length=1, radius=1, shapeName=Dyn_ArmFlap_2r_capsule, center=<0.5 0 0>, rotation=<rot 90 0 0>}]
 Dyn_Bandage_r1 [{shapeType=Capsule, length=2, radius=2, shapeName=Dyn_Bandage_r1_capsule, center=<1 0 0>, rotation=<rot 90 0 0>}]
 Dyn_Bandage_r2 [{shapeType=Capsule, length=2, radius=2, shapeName=Dyn_Bandage_r2_capsule, center=<1 0 0>, rotation=<rot 90 0 0>}]
 Dyn_Bandage_r3 [{shapeType=Capsule, length=2, radius=2, shapeName=Dyn_Bandage_r3_capsule, center=<1 0 0>, rotation=<rot 90 0 0>}]
 Dyn_Master_Bandage_r [{shapeType=Capsule, length=0.5, radius=0.5, shapeName=Dyn_Master_Bandage_r_capsule}]
 lowerarm_twist_01_r [{shapeType=Capsule, length=10, radius=4.5, shapeName=lowerarm_twist_01_r_capsule, center=<-1.9538 -0.8842 0>, rotation=<rot 90 0 0>}]
 lowerarm_twist_02_r [{shapeType=Capsule, length=0.505, radius=0.505, shapeName=lowerarm_twist_02_r_capsule}]
}
constraints [
 {bone1Name=Dyn_Bandage_r1, bone2Name=Dyn_Master_Bandage_r, swing1LimitDegrees=160, swing2LimitDegrees=45, twistLimitDegrees=45}
 {bone1Name=Dyn_Bandage_r2, bone2Name=Dyn_Bandage_r1, swing1LimitDegrees=45, swing2LimitDegrees=45, twistLimitDegrees=45}
 {bone1Name=Dyn_Bandage_r3, bone2Name=Dyn_Bandage_r2, swing1LimitDegrees=45, swing2LimitDegrees=45, twistLimitDegrees=45}
 {bone1Name=Dyn_ArmFlap_1r, bone2Name=lowerarm_twist_02_r, swing1LimitDegrees=20, swing2LimitDegrees=5, twistLimitDegrees=1}
 {bone1Name=Dyn_ArmFlap_2r, bone2Name=lowerarm_twist_02_r, swing1LimitDegrees=20, swing2LimitDegrees=5, twistLimitDegrees=1}
 {
  bone1Name=Dyn_Master_Bandage_r
  bone2Name=lowerarm_twist_01_r
  swing1LimitDegrees=45
  swing1Motion=Free
  swing2LimitDegrees=45
  swing2Motion=Free
  twistLimitDegrees=45
  twistMotion=Free
 }
]
