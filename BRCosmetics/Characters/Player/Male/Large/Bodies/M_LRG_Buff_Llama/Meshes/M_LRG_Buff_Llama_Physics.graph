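# Physics asset for the large male Buff Llama character body, covering dangling accessories: a back pouch, two grenades and a right shoulder flap.
# Capsule bodies on the pouch and grenade bones are constrained to tiny "Master" anchor bodies with narrow twist and asymmetric swing limits.
# The shoulder flap links to the right upper arm with free swing and twist.
# Left and right leg pouch anchors exist without attached bodies or constraints.
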PhysicsAsset /BRCosmetics/Characters/Player/Male/Large/Bodies/M_LRG_Buff_Llama/Meshes/M_LRG_Buff_Llama_Physics
bodies {
 Dyn_BackPouch [{shapeType=Capsule, length=6, radius=4, shapeName=Dyn_BackPouch_capsule, center=<3.5102 0.0001 0>}]
 Dyn_Grenade1 [{shapeType=Capsule, length=5, radius=2, shapeName=Dyn_Grenade1_capsule, center=<3.3803 0 -0.0001>, rotation=<rot 90 0 0>}]
 Dyn_Grenade2 [{shapeType=Capsule, length=5, radius=2, shapeName=Dyn_Grenade2_capsule, center=<3.3803 0 -0.0001>, rotation=<rot 90 0 0>}]
 Dyn_Master_BackPouch [{shapeType=Capsule, length=0.5, radius=0.5, shapeName=Dyn_Master_BackPouch_capsule}]
 Dyn_Master_Grenade1 [{shapeType=Capsule, length=0.5, radius=0.5, shapeName=Dyn_Master_Grenade1_capsule}]
 Dyn_Master_Grenade2 [{shapeType=Capsule, length=0.5, radius=0.5, shapeName=Dyn_Master_Grenade2_capsule}]
 Dyn_Master_LegPouch_l [{shapeType=Capsule, length=0.5, radius=0.5, shapeName=Dyn_Master_LegPouch_l_capsule}]
 Dyn_Master_LegPouch_r [{shapeType=Capsule, length=0.5, radius=0.5, shapeName=Dyn_Master_LegPouch_r_capsule}]
 Dyn_Shoulder_Flap_r [{shapeType=Capsule, length=1, radius=1, shapeName=Dyn_Shoulder_Flap_r_capsule, center=<1 0 0>}]
 upperarm_r [{shapeType=Capsule, length=1, radius=1, shapeName=upperarm_r_capsule}]
}
constraints [
 {
  bone1Name=Dyn_Shoulder_Flap_r
  bone2Name=upperarm_r
  swing1LimitDegrees=45
  swing1Motion=Free
  swing2LimitDegrees=45
  swing2Motion=Free
  twistLimitDegrees=45
  twistMotion=Free
 }
 {bone1Name=Dyn_BackPouch, bone2Name=Dyn_Master_BackPouch, swing1LimitDegrees=5, swing2LimitDegrees=30, twistLimitDegrees=1}
 {bone1Name=Dyn_Grenade2, bone2Name=Dyn_Master_Grenade2, swing1LimitDegrees=40, swing2LimitDegrees=20, twistLimitDegrees=1}
 {bone1Name=Dyn_Grenade1, bone2Name=Dyn_Master_Grenade1, swing1LimitDegrees=40, swing2LimitDegrees=20, twistLimitDegrees=1}
]
