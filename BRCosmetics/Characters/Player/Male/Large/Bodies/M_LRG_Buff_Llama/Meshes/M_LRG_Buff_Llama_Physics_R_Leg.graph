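# Physics asset for the large male Buff Llama player body, named for the right leg, with capsule bodies on dynamic accessory bones.
# The right leg pouch carries the largest capsule and is angularly locked to its master bone.
# The right shoulder flap rotates freely against the upper arm, leaving its angular limit values unused.
# Small capsules on the back pouch, grenade and left leg pouch master bones have no constraints.
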PhysicsAsset /BRCosmetics/Characters/Player/Male/Large/Bodies/M_LRG_Buff_Llama/Meshes/M_LRG_Buff_Llama_Physics_R_Leg
bodies {
 Dyn_LegPouch_1r [{shapeType=Capsule, length=1, radius=5, shapeName=Dyn_LegPouch_1r_capsule, center=<1 0 0>}]
 Dyn_Master_BackPouch [{shapeType=Capsule, length=0.5, radius=0.5, shapeName=Dyn_Master_BackPouch_capsule}]
 Dyn_Master_Grenade1 [{shapeType=Capsule, length=0.5, radius=0.5, shapeName=Dyn_Master_Grenade1_capsule}]
 Dyn_Master_Grenade2 [{shapeType=Capsule, length=0.5, radius=0.5, shapeName=Dyn_Master_Grenade2_capsule}]
 Dyn_Master_LegPouch_l [{shapeType=Capsule, length=0.5, radius=0.5, shapeName=Dyn_Master_LegPouch_l_capsule}]
 Dyn_Master_LegPouch_r [{shapeType=Capsule, length=0.5, radius=0.5, shapeName=Dyn_Master_LegPouch_r_capsule}]
 Dyn_Shoulder_Flap_r [{shapeType=Capsule, length=1, radius=1, shapeName=Dyn_Shoulder_Flap_r_capsule, center=<1 0 0>}]
 upperarm_r [{shapeType=Capsule, length=1, radius=1, shapeName=upperarm_r_capsule}]
}
constraints [
 {
  bone1Name=Dyn_Shoulder_Flap_r
  bone2Name=upperarm_r
  swing1LimitDegrees=45
  swing1Motion=Free
  swing2LimitDegrees=45
  swing2Motion=Free
  twistLimitDegrees=45
  twistMotion=Free
 }
 {
  bone1Name=Dyn_LegPouch_1r
  bone2Name=Dyn_Master_LegPouch_r
  swing1LimitDegrees=45
  swing1Motion=Locked
  swing2LimitDegrees=45
  swing2Motion=Locked
  twistLimitDegrees=45
  twistMotion=Locked
 }
]
