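# Animation Blueprint for the Buff Llama large male body cosmetic, adding secondary physics motion over the pose copied from the attached parent mesh.
# Six chained RigidBody nodes simulate main, left and right arm bandage, and per-leg physics assets in pelvis or thigh space, with scaled-down component motion and overridden gravity.
# The final front-end physics node is enabled only when the inherited bIsInFrontEnd is true; the event graph holds no logic.
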AnimBlueprint /BRCosmetics/Characters/Player/Male/Large/Bodies/M_LRG_Buff_Llama/Meshes/M_LRG_Buff_Llama_AnimBP
target_skeleton=&/M_LRG_Buff_Llama_Skeleton
graphs [
 AnimGraph {nodes [
   n1 = RigidBody(ComponentPose=Local To Component(LocalPose=Copy Pose From Mesh() {bUseAttachedParent=true})) {
    bClampLinearTranslationLimitToRefPose=true
    bFreezeIncomingPoseOnStart=true
    bOverrideWorldGravity=true
    overridePhysicsAsset=&/M_LRG_Buff_Llama_Physics
    simulationSpace=BaseBoneSpace
    alphaBoolBlend {blendOption=Linear}
    baseBoneRef {boneName=pelvis}
    componentLinearAccScale=<0.1 0.5 0.1>
    componentLinearVelScale=<0.1 0.5 0.1>
    overrideWorldGravity {z=-980}
   }
   n2 = RigidBody(ComponentPose=@n1) {
    bClampLinearTranslationLimitToRefPose=true
    bOverrideWorldGravity=true
    overridePhysicsAsset=&/M_LRG_Buff_Llama_Bandage_Physics_Left_Arm
    simulationSpace=BaseBoneSpace
    alphaBoolBlend {blendOption=Linear}
    baseBoneRef {boneName=pelvis}
    componentLinearAccScale=<0.2 0.1 0.1>
    componentLinearVelScale=<0.2 0.1 0.1>
    overrideWorldGravity {z=-980}
   }
   n3 = RigidBody(ComponentPose=@n2) {
    bClampLinearTranslationLimitToRefPose=true
    bFreezeIncomingPoseOnStart=true
    bOverrideWorldGravity=true
    overridePhysicsAsset=&/M_LRG_Buff_Llama_Physics_L_Leg
    simulationSpace=BaseBoneSpace
    alphaBoolBlend {blendOption=Linear}
    baseBoneRef {boneName=thigh_l}
    componentLinearAccScale=<0.1 0.1 0.1>
    componentLinearVelScale=<0.1 0.1 0.1>
    overrideWorldGravity {z=-980}
   }
   n4 = RigidBody(ComponentPose=@n3) {
    bClampLinearTranslationLimitToRefPose=true
    bOverrideWorldGravity=true
    overridePhysicsAsset=&/M_LRG_Buff_Llama_Bangage_Physics_Right_Arm
    simulationSpace=BaseBoneSpace
    alphaBoolBlend {blendOption=Linear}
    baseBoneRef {boneName=pelvis}
    componentLinearAccScale=<0.2 0.1 0.1>
    componentLinearVelScale=<0.2 0.1 0.1>
    overrideWorldGravity {z=-980}
   }
   n5 = RigidBody(ComponentPose=@n4) {
    bClampLinearTranslationLimitToRefPose=true
    bFreezeIncomingPoseOnStart=true
    bOverrideWorldGravity=true
    overridePhysicsAsset=&/M_LRG_Buff_Llama_Physics_R_Leg
    simulationSpace=BaseBoneSpace
    alphaBoolBlend {blendOption=Linear}
    baseBoneRef {boneName=thigh_r}
    componentLinearAccScale=<0.1 0.1 0.1>
    componentLinearVelScale=<0.1 0.1 0.1>
    overrideWorldGravity {z=-980}
   }
   n6 = RigidBody!(ComponentPose=@n5, bAlphaBoolEnabled:bool=Get bIsInFrontEnd(), ExternalForce) -> (Pose) {
    alphaInputType=Bool
    bClampLinearTranslationLimitToRefPose=true
    bFreezeIncomingPoseOnStart=true
    bOverrideWorldGravity=true
    overridePhysicsAsset=&/M_LRG_Buff_Llama_FrontEnd_Physics
    simulationSpace=BaseBoneSpace
    alphaBoolBlend {blendOption=Linear}
    baseBoneRef {boneName=pelvis}
    componentAppliedLinearAccClamp=<3000 3000 3000>
    componentLinearAccScale=<0.01 0.01 0.01>
    componentLinearVelScale=<0.01 0.01 0.01>
    overrideWorldGravity {z=-980}
   }
   Output Pose(Result=Component To Local(ComponentPose=@n6)) {layerGroup=None}
  ]}
 EventGraph {nodes [Event BlueprintUpdateAnimation()]}
]
variables []
