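# Control Rig that procedurally drives limb twist bones for the large buff llama character's face part.
# Upper arms and thighs extract X-axis swing-twist and apply negated angles to their twist bones, with thighs measured from the initial pose.
# Forearm and calf twist bones are constrained toward hand and foot, reduced to X rotation, then half is passed to each second twist bone.
# Many targeted limb bones are absent from this head-and-upper-body hierarchy; a debug hierarchy draw ends the chain.
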ControlRigBlueprint /BRCosmetics/Characters/Player/Male/Large/Bodies/M_LRG_Buff_Llama/Meshes/Parts/M_LRG_Buff_LlamaFace_Deform_CtrlRig_twists
graphs [
 RigVMFunctionLibrary {links [], nodes []}
 RigVMModel {nodes [
   RigUnit_BeginExecution = &RigUnit_BeginExecution()
   n1 = /Script/RigVM.RigVMFunction_MathQuaternionToAxisAndAngle(Value=&RigVMFunction_MathQuaternionSwingTwist(TwistAxis=(X=1,Y=0,Z=0), Input=&RigUnit_GetBoneTransform(Bone=upperarm_l, Space=LocalSpace).Transform.Rotation).Twist) {is=struct RigVMUnitNode, template="ToAxisAndAngle::Execute(in Value,out Axis,out Angle)"}
   RerouteNode_2_1 = RigVMDispatchNode(Value=-0.7) {template="DISPATCH_RigVMDispatch_Constant(io Value)"}
   /Script/ControlRig.RigUnit_SetBoneRotation(
    Bone=upperarm_twist_01_l
    Weight=1
    bPropagateToChildren=True
    Rotation=/Script/RigVM.RigVMFunction_MathQuaternionFromAxisAndAngle(
     Angle=&RigVMFunction_MathFloatMul(A=@n1.Angle, B=@RerouteNode_2_1.Value) {template="Multiply::Execute(in A,in B,out Result)"}.Result
     Axis=@n1.Axis
    ) {is=struct RigVMUnitNode, template="FromAxisAndAngle::Execute(in Axis,in Angle,out Result)"}.Result
    ExecutePin=@RigUnit_BeginExecution.ExecutePin
   ) {is=struct RigVMUnitNode}
   n2 = &RigVMFunction_MathQuaternionToAxisAndAngle(Value=&RigVMFunction_MathQuaternionSwingTwist(TwistAxis=(X=1,Y=0,Z=0), Input=&RigUnit_GetBoneTransform(Bone=upperarm_l, Space=LocalSpace).Transform.Rotation).Twist) {template="ToAxisAndAngle::Execute(in Value,out Axis,out Angle)"}
   RerouteNode_2 = RigVMDispatchNode(Value=-0.6) {template="DISPATCH_RigVMDispatch_Constant(io Value)"}
   > &RigUnit_SetBoneRotation(
    Bone=upperarm_twist_02_l
    Weight=1
    bPropagateToChildren=True
    Rotation=&RigVMFunction_MathQuaternionFromAxisAndAngle(
     Angle=&RigVMFunction_MathFloatMul(A=@n2.Angle, B=@RerouteNode_2.Value) {template="Multiply::Execute(in A,in B,out Result)"}.Result
     Axis=@n2.Axis
    ) {template="FromAxisAndAngle::Execute(in Axis,in Angle,out Result)"}.Result
   )
   n3 = &RigVMFunction_MathQuaternionToAxisAndAngle(Value=&RigVMFunction_MathQuaternionSwingTwist(TwistAxis=(X=1,Y=0,Z=0), Input=&RigUnit_GetBoneTransform(Bone=upperarm_r, Space=LocalSpace).Transform.Rotation).Twist) {template="ToAxisAndAngle::Execute(in Value,out Axis,out Angle)"}
   > &RigUnit_SetBoneRotation(
    Bone=upperarm_twist_01_r
    Weight=1
    bPropagateToChildren=True
    Rotation=&RigVMFunction_MathQuaternionFromAxisAndAngle(
     Angle=&RigVMFunction_MathFloatMul(A=@n3.Angle, B=@RerouteNode_2_1.Value) {template="Multiply::Execute(in A,in B,out Result)"}.Result
     Axis=@n3.Axis
    ) {template="FromAxisAndAngle::Execute(in Axis,in Angle,out Result)"}.Result
   )
   n4 = &RigVMFunction_MathQuaternionToAxisAndAngle(Value=&RigVMFunction_MathQuaternionSwingTwist(TwistAxis=(X=1,Y=0,Z=0), Input=&RigUnit_GetBoneTransform(Bone=upperarm_r, Space=LocalSpace).Transform.Rotation).Twist) {template="ToAxisAndAngle::Execute(in Value,out Axis,out Angle)"}
   > &RigUnit_SetBoneRotation(
    Bone=upperarm_twist_02_r
    Weight=1
    bPropagateToChildren=True
    Rotation=&RigVMFunction_MathQuaternionFromAxisAndAngle(
     Angle=&RigVMFunction_MathFloatMul(A=@n4.Angle, B=@RerouteNode_2.Value) {template="Multiply::Execute(in A,in B,out Result)"}.Result
     Axis=@n4.Axis
    ) {template="FromAxisAndAngle::Execute(in Axis,in Angle,out Result)"}.Result
   )
   n5 = &RigUnit_GetBoneTransform(Bone=thigh_l, Space=LocalSpace)
   n6 = /Script/ControlRig.RigUnit_GetInitialBoneTransform(Bone=thigh_l, Space=LocalSpace) {is=struct RigVMUnitNode}
   n7 = &RigVMFunction_MathQuaternionToAxisAndAngle(Value=&RigVMFunction_MathQuaternionSwingTwist(
     TwistAxis=(X=1,Y=0,Z=0)
     Input=&RigVMFunction_MathTransformMakeRelative(Global=@n5.Transform, Parent=@n6.Transform).Local.Rotation
    ).Twist) {template="ToAxisAndAngle::Execute(in Value,out Axis,out Angle)"}
   > &RigUnit_SetBoneRotation(
    Bone=thigh_twist_01_l
    Weight=1
    bPropagateToChildren=True
    Rotation=&RigVMFunction_MathQuaternionFromAxisAndAngle(Angle=&RigVMFunction_MathFloatMul(B=-1, A=@n7.Angle) {template="Multiply::Execute(in A,in B,out Result)"}.Result, Axis=@n7.Axis) {template="FromAxisAndAngle::Execute(in Axis,in Angle,out Result)"}.Result
   )
   n8 = &RigUnit_GetBoneTransform(Bone=thigh_r, Space=LocalSpace)
   RigUnit_GetInitialBoneTransform = &RigUnit_GetInitialBoneTransform(Bone=thigh_r, Space=LocalSpace)
   n9 = &RigVMFunction_MathQuaternionToAxisAndAngle(Value=&RigVMFunction_MathQuaternionSwingTwist(
     TwistAxis=(X=1,Y=0,Z=0)
     Input=&RigVMFunction_MathTransformMakeRelative(Global=@n8.Transform, Parent=@RigUnit_GetInitialBoneTransform.Transform).Local.Rotation
    ).Twist) {template="ToAxisAndAngle::Execute(in Value,out Axis,out Angle)"}
   > &RigUnit_SetBoneRotation(
    Bone=thigh_twist_01_r
    Space=LocalSpace
    Weight=1
    bPropagateToChildren=True
    Rotation=&RigVMFunction_MathQuaternionFromAxisAndAngle(Angle=&RigVMFunction_MathFloatMul(B=-1, A=@n9.Angle) {template="Multiply::Execute(in A,in B,out Result)"}.Result, Axis=@n9.Axis) {template="FromAxisAndAngle::Execute(in Axis,in Angle,out Result)"}.Result
   )
   RigUnit_GetBoneTransform = &RigUnit_GetBoneTransform(Bone=lowerarm_twist_01_l)
   > /Script/ControlRig.RigUnit_TransformConstraint(
    BaseBone=None
    BaseTransform=(Rotation=(X=0,Y=0,Z=0,W=1),Translation=(X=0,Y=0,Z=0),Scale3D=(X=1,Y=1,Z=1))
    BaseTransformSpace=GlobalSpace
    Bone=lowerarm_twist_01_l
    Targets="((Transform=(Rotation=(X=0,Y=0,Z=0,W=1),Translation=(X=0,Y=0,Z=0),Scale3D=(X=1,Y=1,Z=1)),Weight=0.75,bMaintainOffset=True,Filter=(TranslationFilter=(bX=false,bY=false,bZ=false),RotationFilter=(bX=True,bY=false,bZ=false),ScaleFilter=(bX=false,bY=false,bZ=fa ...(truncated)"
    bUseInitialTransforms=True
    Targets.1.Transform=@RigUnit_GetBoneTransform.Transform
    Targets.0.Transform=&RigUnit_GetBoneTransform(Bone=hand_l).Transform
   ) {is=struct RigVMUnitNode}
   n10 = &RigVMFunction_MathQuaternionToEuler(RotationOrder=ZYX, Value=&RigUnit_GetBoneTransform(Bone=lowerarm_twist_01_l, Space=LocalSpace).Transform.Rotation)
   > &RigUnit_SetBoneRotation(
    Bone=lowerarm_twist_01_l
    Weight=1
    bPropagateToChildren=True
    Rotation=&RigVMFunction_MathQuaternionFromEuler(Euler=(X=0,Y=0,Z=0), Euler.X=@n10.Result.X).Result
   )
   > &RigUnit_SetBoneRotation(
    Bone=lowerarm_twist_02_l
    Weight=1
    bPropagateToChildren=True
    Rotation=&RigVMFunction_MathQuaternionFromEuler(
     Euler=(X=0,Y=0,Z=0)
     Euler.X=&RigVMFunction_MathFloatMul(B=0.5, A=@n10.Result.X) {template="Multiply::Execute(in A,in B,out Result)"}.Result
    ).Result
   )
   > &RigUnit_TransformConstraint(
    BaseBone=None
    BaseTransform=(Rotation=(X=0,Y=0,Z=0,W=1),Translation=(X=0,Y=0,Z=0),Scale3D=(X=1,Y=1,Z=1))
    BaseTransformSpace=GlobalSpace
    Bone=lowerarm_twist_01_r
    Targets="((Transform=(Rotation=(X=0,Y=0,Z=0,W=1),Translation=(X=0,Y=0,Z=0),Scale3D=(X=1,Y=1,Z=1)),Weight=0.75,bMaintainOffset=True,Filter=(TranslationFilter=(bX=false,bY=false,bZ=false),RotationFilter=(bX=True,bY=false,bZ=false),ScaleFilter=(bX=false,bY=false,bZ=fa ...(truncated)"
    bUseInitialTransforms=True
    Targets.0.Transform.Rotation=&RigUnit_GetBoneTransform(Bone=hand_r).Transform.Rotation
    Targets.1.Transform.Rotation=&RigUnit_GetBoneTransform(Bone=lowerarm_twist_01_r).Transform.Rotation
   )
   n11 = &RigVMFunction_MathQuaternionToEuler(RotationOrder=ZYX, Value=&RigUnit_GetBoneTransform(Bone=lowerarm_twist_01_r, Space=LocalSpace).Transform.Rotation)
   > RigUnit_SetBoneRotation = &RigUnit_SetBoneRotation(
    Bone=lowerarm_twist_01_r
    Weight=1
    bPropagateToChildren=True
    Rotation=&RigVMFunction_MathQuaternionFromEuler(Euler=(X=0,Y=0,Z=0), Euler.X=@n11.Result.X).Result
   )
   > &RigUnit_SetBoneRotation(
    Bone=lowerarm_twist_02_r
    Weight=1
    bPropagateToChildren=True
    Rotation=&RigVMFunction_MathQuaternionFromEuler(
     Euler=(X=0,Y=0,Z=0)
     Euler.X=&RigVMFunction_MathFloatMul(B=0.5, A=@n11.Result.X) {template="Multiply::Execute(in A,in B,out Result)"}.Result
    ).Result
   )
   > &RigUnit_TransformConstraint(
    BaseBone=None
    BaseTransform=(Rotation=(X=0,Y=0,Z=0,W=1),Translation=(X=0,Y=0,Z=0),Scale3D=(X=1,Y=1,Z=1))
    BaseTransformSpace=GlobalSpace
    Bone=calf_twist_01_l
    Targets="((Transform=(Rotation=(X=0,Y=0,Z=0,W=1),Translation=(X=0,Y=0,Z=0),Scale3D=(X=1,Y=1,Z=1)),Weight=0.75,bMaintainOffset=True,Filter=(TranslationFilter=(bX=false,bY=false,bZ=false),RotationFilter=(bX=True,bY=false,bZ=false),ScaleFilter=(bX=false,bY=false,bZ=fa ...(truncated)"
    bUseInitialTransforms=True
    Targets.0.Transform=&RigUnit_GetBoneTransform(Bone=foot_l).Transform
    Targets.1.Transform=&RigUnit_GetBoneTransform(Bone=calf_twist_01_l).Transform
   )
   n12 = &RigVMFunction_MathQuaternionToEuler(RotationOrder=ZYX, Value=&RigUnit_GetBoneTransform(Bone=calf_twist_01_l, Space=LocalSpace).Transform.Rotation)
   > &RigUnit_SetBoneRotation(
    Bone=calf_twist_01_l
    Weight=1
    bPropagateToChildren=True
    Rotation=&RigVMFunction_MathQuaternionFromEuler(Euler=(X=0,Y=0,Z=0), Euler.X=@n12.Result.X).Result
   )
   > &RigUnit_SetBoneRotation(
    Bone=calf_twist_02_l
    Weight=1
    bPropagateToChildren=True
    Rotation=&RigVMFunction_MathQuaternionFromEuler(
     Euler=(X=0,Y=0,Z=0)
     Euler.X=&RigVMFunction_MathFloatMul(B=0.5, A=@n12.Result.X) {template="Multiply::Execute(in A,in B,out Result)"}.Result
    ).Result
   )
   > &RigUnit_TransformConstraint(
    BaseBone=None
    BaseTransform=(Rotation=(X=0,Y=0,Z=0,W=1),Translation=(X=0,Y=0,Z=0),Scale3D=(X=1,Y=1,Z=1))
    BaseTransformSpace=GlobalSpace
    Bone=calf_twist_01_r
    Targets="((Transform=(Rotation=(X=0,Y=0,Z=0,W=1),Translation=(X=0,Y=0,Z=0),Scale3D=(X=1,Y=1,Z=1)),Weight=0.75,bMaintainOffset=True,Filter=(TranslationFilter=(bX=false,bY=false,bZ=false),RotationFilter=(bX=True,bY=false,bZ=false),ScaleFilter=(bX=false,bY=false,bZ=fa ...(truncated)"
    bUseInitialTransforms=True
    Targets.0.Transform=&RigUnit_GetBoneTransform(Bone=foot_r).Transform
    Targets.1.Transform=&RigUnit_GetBoneTransform(Bone=calf_twist_01_r).Transform
   )
   n13 = &RigVMFunction_MathQuaternionToEuler(RotationOrder=ZYX, Value=&RigUnit_GetBoneTransform(Bone=calf_twist_01_r, Space=LocalSpace).Transform.Rotation)
   > &RigUnit_SetBoneRotation(
    Bone=calf_twist_01_r
    Weight=1
    bPropagateToChildren=True
    Rotation=&RigVMFunction_MathQuaternionFromEuler(Euler=(X=0,Y=0,Z=0), Euler.X=@n13.Result.X).Result
   )
   > &RigUnit_SetBoneRotation(
    Bone=calf_twist_02_r
    Weight=1
    bPropagateToChildren=True
    Rotation=&RigVMFunction_MathQuaternionFromEuler(
     Euler=(X=0,Y=0,Z=0)
     Euler.X=&RigVMFunction_MathFloatMul(B=0.5, A=@n13.Result.X) {template="Multiply::Execute(in A,in B,out Result)"}.Result
    ).Result
   )
   > /Script/ControlRig.RigUnit_DebugHierarchy(
    Color=(R=1,G=1,B=1,A=1)
    DebugDrawSettings=(DepthPriority=SDPG_Foreground,Lifetime=-1)
    Items=()
    Scale=5
    Thickness=0.2
    WorldOffset=(Rotation=(X=0,Y=0,Z=0,W=1),Translation=(X=0,Y=0,Z=0),Scale3D=(X=1,Y=1,Z=1))
    bEnabled=True
   ) {is=struct RigVMUnitNode}
   RigVMVariableNode() {variable=Upperarm_Twist}
  ]}
]
hierarchy [
 Root {loc=<0 0 0>, rot=<0 0 0 1>}
 pelvis {parent=Root, loc=<0 -0.7321 92.4218>, rot=<0.0127 0.707 -0.0127 -0.707>}
 Spine_01 {parent=pelvis, loc=<5.6002 -0.5128 0>, rot=<0 0 0.034 -0.9994>}
 Spine_02 {parent=Spine_01, loc=<8.0753 0.0139 0>, rot=<0 0 0.0238 -0.9997>}
 Spine_03 {parent=Spine_02, loc=<7.0393 0 0>, rot=<0 0 0.0149 -0.9999>}
 spine_04 {parent=Spine_03, loc=<8.1023 0 0>, rot=<0 0 -0.0132 -0.9999>}
 spine_05 {parent=spine_04, loc=<10.0146 0 0>, rot=<0 0 -0.0015 -1>}
 clavicle_l {parent=spine_05, loc=<2.5856 -0.2866 -4.3777>, rot=<0.0465 -0.7289 -0.009 -0.6829>}
 upperarm_l {parent=clavicle_l, loc=<18.2264 0.1342 -0.001>, rot=<0.0253 -0.2693 0.0013 -0.9627>}
 upperarm_twist_01_l {parent=upperarm_l, loc=<7.9796 0.0843 -0.6195>, rot=<0 0 0 1>}
 deform_pectoral_fr_l {parent=clavicle_l, loc=<12.1058 0.1342 -0.001>, rot=<0 0 0 1>}
 clavicle_r {parent=spine_05, loc=<2.5856 -0.2865 4.3777>, rot=<0.7289 0.0465 -0.6829 0.009>}
 upperarm_r {parent=clavicle_r, loc=<-18.2264 -0.1342 0.001>, rot=<0.0253 -0.2693 0.0013 -0.9627>}
 upperarm_twist_01_r {parent=upperarm_r, loc=<-7.9796 -0.0843 0.6195>, rot=<0 0 0 1>}
 deform_pectoral_fr_r {parent=clavicle_r, loc=<-12.1058 -0.1342 0.0012>, rot=<0 0 0 1>}
 neck_01 {parent=spine_05, loc=<10.815 1.9196 0>, rot=<0 0 -0.1864 -0.9825>}
 neck_02 {parent=neck_01, loc=<4.7347 0.3045 0>, rot=<0 0 0.0363 -0.9993>}
 head {parent=neck_02, loc=<5.7587 0.5544 0.0004>, rot=<0.1622 0.0665 0.4068 -0.8965>}
 Dyn_Master_Mouth {parent=head, loc=<3.1583 11.8078 0>, rot=<0 0 -0.8417 -0.5399>}
 Dyn_Mouth {parent=Dyn_Master_Mouth, loc=<0.4944 0 0>, rot=<0 0 0 1>}
 Dyn_Mouth_Flap1 {parent=Dyn_Mouth, loc=<4.4202 2.1501 -0.5379>, rot=<0 0 -0.3607 -0.9327>}
 Dyn_Mouth_Flap2 {parent=Dyn_Mouth, loc=<3.7789 3.0225 0.5987>, rot=<0 0 -0.5494 -0.8356>}
 Dyn_Mouth_Flap3 {parent=Dyn_Mouth, loc=<2.54 2.7116 0.802>, rot=<0 0 -0.7938 -0.6081>}
 Dyn_Flap1 {parent=head, loc=<4.7665 1.7574 -7.2434>, rot=<-0.2516 0.6287 0.7279 0.1075>}
 Dyn_Flap2 {parent=head, loc=<5.4893 -0.6142 -7.7287>, rot=<-0.4034 0.5927 0.6968 0.0193>}
 Dyn_Flap3 {parent=head, loc=<2.8512 1.2617 -6.7864>, rot=<-0.1923 0.6181 0.7458 0.1577>}
 Dyn_Flap4 {parent=head, loc=<3.4314 -0.9047 -7.8725>, rot=<-0.387 0.6377 0.6658 -0.0155>}
 Dyn_Flap5 {parent=head, loc=<4.1731 -3.3672 -7.9979>, rot=<0.4945 -0.6258 -0.5904 0.1236>}
 Dyn_Flap6 {parent=head, loc=<-1.5855 -2.313 -8.7804>, rot=<-0.4048 0.5451 0.6772 0.2834>}
 Dyn_Flap7 {parent=head, loc=<-0.2509 -5.7083 -8.6633>, rot=<-0.5089 0.5838 0.6029 0.1916>}
 Dyn_Flap8 {parent=head, loc=<4.7227 -12.0203 -0.8105>, rot=<0.4056 -0.9114 -0.0536 -0.0443>}
 Dyn_Flap9 {parent=head, loc=<5.5014 -9.9242 -6.0977>, rot=<0.3492 -0.861 -0.339 -0.148>}
 Dyn_Flap10 {parent=head, loc=<5.4269 -11.4013 2.4644>, rot=<0.3955 -0.9118 0.0387 0.1033>}
 Dyn_Flap11 {parent=head, loc=<0.6707 -7.2338 8.4445>, rot=<0.6484 -0.4685 0.4911 0.3447>}
 Dyn_Flap12 {parent=head, loc=<-0.7004 -3.9182 8.8673>, rot=<0.5675 -0.5638 0.5388 0.264>}
 Dyn_Flap13 {parent=head, loc=<-1.2265 -0.9187 8.2563>, rot=<0.426 -0.4866 0.6565 0.3884>}
 Dyn_Flap14 {parent=head, loc=<-0.1787 7.1433 -2.207>, rot=<0 0 0.9623 0.2718>}
 Dyn_Ears {parent=head, loc=<21.1204 -2.6768 0.0375>, rot=<-0.0152 0.0089 -0.2809 -0.9596>}
 Dyn_Ears1 {parent=head, loc=<21.0948 -2.7792 3.0333>, rot=<-0.0152 0.0089 -0.2809 -0.9596>}
 attach {parent=Root, loc=<0 0 0>, rot=<0 0 0 1>}
 NewControl {type=Control, loc=<0 0 0>, rot=<0 0 0 1>}
 r_knee_corrective {type=Curve, loc=<0 0 0>, rot=<0 0 0 1>}
 l_knee_corrective {type=Curve, loc=<0 0 0>, rot=<0 0 0 1>}
 Male_Med_Head_01_Pout {type=Curve, loc=<0 0 0>, rot=<0 0 0 1>}
 Male_Med_Head_02_Pout {type=Curve, loc=<0 0 0>, rot=<0 0 0 1>}
 Male_Med_Head_03_Pout {type=Curve, loc=<0 0 0>, rot=<0 0 0 1>}
 Male_Med_Head_05_Effort {type=Curve, loc=<0 0 0>, rot=<0 0 0 1>}
 Stretch {type=Curve, loc=<0 0 0>, rot=<0 0 0 1>}
 Squash {type=Curve, loc=<0 0 0>, rot=<0 0 0 1>}
 polySurface151 {type=Curve, loc=<0 0 0>, rot=<0 0 0 1>}
 l_cheek_puff {type=Curve, loc=<0 0 0>, rot=<0 0 0 1>}
 r_cheek_puff {type=Curve, loc=<0 0 0>, rot=<0 0 0 1>}
 l_smile_fix {type=Curve, loc=<0 0 0>, rot=<0 0 0 1>}
 r_smile_fix {type=Curve, loc=<0 0 0>, rot=<0 0 0 1>}
 jaw_open_fix {type=Curve, loc=<0 0 0>, rot=<0 0 0 1>}
 r_lip_sneer_fix {type=Curve, loc=<0 0 0>, rot=<0 0 0 1>}
 l_lip_sneer_fix {type=Curve, loc=<0 0 0>, rot=<0 0 0 1>}
 L_Blink {type=Curve, loc=<0 0 0>, rot=<0 0 0 1>}
 R_Blink {type=Curve, loc=<0 0 0>, rot=<0 0 0 1>}
 L_Arm_Stretch {type=Curve, loc=<0 0 0>, rot=<0 0 0 1>}
 Hair_Hat {type=Curve, loc=<0 0 0>, rot=<0 0 0 1>}
 Hair_Helmet {type=Curve, loc=<0 0 0>, rot=<0 0 0 1>}
 Male_Med_Head_02_What {type=Curve, loc=<0 0 0>, rot=<0 0 0 1>}
 Male_Med_Head_02_Smile {type=Curve, loc=<0 0 0>, rot=<0 0 0 1>}
 Male_Med_Head_02_Scowl {type=Curve, loc=<0 0 0>, rot=<0 0 0 1>}
 Male_Med_Head_02_Hmmph {type=Curve, loc=<0 0 0>, rot=<0 0 0 1>}
 Male_Med_Head_02_Grrr {type=Curve, loc=<0 0 0>, rot=<0 0 0 1>}
 Male_Med_Head_02_Effort {type=Curve, loc=<0 0 0>, rot=<0 0 0 1>}
 Male_Med_Head_03_What {type=Curve, loc=<0 0 0>, rot=<0 0 0 1>}
 Male_Med_Head_03_Grrr {type=Curve, loc=<0 0 0>, rot=<0 0 0 1>}
 Male_Med_Head_03_Effort {type=Curve, loc=<0 0 0>, rot=<0 0 0 1>}
 Male_Med_Head_03_Hmmph {type=Curve, loc=<0 0 0>, rot=<0 0 0 1>}
 Male_Med_Head_03_Scowl {type=Curve, loc=<0 0 0>, rot=<0 0 0 1>}
 Male_Med_Head_03_Smile {type=Curve, loc=<0 0 0>, rot=<0 0 0 1>}
 Male_Med_Head_05_Teeth_Low_What {type=Curve, loc=<0 0 0>, rot=<0 0 0 1>}
 Male_Med_Head_05_Teeth_Low_Pout {type=Curve, loc=<0 0 0>, rot=<0 0 0 1>}
 Male_Med_Head_05_Teeth_Low_Hmmph {type=Curve, loc=<0 0 0>, rot=<0 0 0 1>}
 Male_Med_Head_05_Teeth_Low_Grrr {type=Curve, loc=<0 0 0>, rot=<0 0 0 1>}
 Male_Med_Head_05_Teeth_Low_Effort {type=Curve, loc=<0 0 0>, rot=<0 0 0 1>}
 Male_Med_Head_05_What {type=Curve, loc=<0 0 0>, rot=<0 0 0 1>}
 Male_Med_Head_05_Grrr {type=Curve, loc=<0 0 0>, rot=<0 0 0 1>}
 Male_Med_Head_05_Hmmph {type=Curve, loc=<0 0 0>, rot=<0 0 0 1>}
 Male_Med_Head_05_Pout {type=Curve, loc=<0 0 0>, rot=<0 0 0 1>}
 Male_Med_Head_05_Tongue_What {type=Curve, loc=<0 0 0>, rot=<0 0 0 1>}
 Male_Med_Head_05_Tongue_Pout {type=Curve, loc=<0 0 0>, rot=<0 0 0 1>}
 Male_Med_Head_05_Tongue_Hmmph {type=Curve, loc=<0 0 0>, rot=<0 0 0 1>}
 Male_Med_Head_05_Tongue_Grrr {type=Curve, loc=<0 0 0>, rot=<0 0 0 1>}
 Male_Med_Head_05_Tongue_Effort {type=Curve, loc=<0 0 0>, rot=<0 0 0 1>}
 AdditiveLayer {type=Curve, loc=<0 0 0>, rot=<0 0 0 1>}
 RightHandIK {type=Curve, loc=<0 0 0>, rot=<0 0 0 1>}
 LeftHandIK {type=Curve, loc=<0 0 0>, rot=<0 0 0 1>}
 LeftHandFK {type=Curve, loc=<0 0 0>, rot=<0 0 0 1>}
 Male_Med_Head_01_Grrr {type=Curve, loc=<0 0 0>, rot=<0 0 0 1>}
 WeaponAiming {type=Curve, loc=<0 0 0>, rot=<0 0 0 1>}
 Male_Med_Head_01_Hmmph {type=Curve, loc=<0 0 0>, rot=<0 0 0 1>}
 DragonSlash {type=Curve, loc=<0 0 0>, rot=<0 0 0 1>}
 Male_Med_Head_01_Effort {type=Curve, loc=<0 0 0>, rot=<0 0 0 1>}
 LeftHankFK {type=Curve, loc=<0 0 0>, rot=<0 0 0 1>}
 blendOrient1 {type=Curve, loc=<0 0 0>, rot=<0 0 0 1>}
 blendParent1 {type=Curve, loc=<0 0 0>, rot=<0 0 0 1>}
 bottom_R_Corner_Down {type=Curve, loc=<0 0 0>, rot=<0 0 0 1>}
 bottom_R_Corner_Up {type=Curve, loc=<0 0 0>, rot=<0 0 0 1>}
 top_L_Corner_Up {type=Curve, loc=<0 0 0>, rot=<0 0 0 1>}
 top_L_Corner_Down {type=Curve, loc=<0 0 0>, rot=<0 0 0 1>}
 top_R_Corner_Down {type=Curve, loc=<0 0 0>, rot=<0 0 0 1>}
 top_R_Corner_Up {type=Curve, loc=<0 0 0>, rot=<0 0 0 1>}
 bottom_L_Corner_Down {type=Curve, loc=<0 0 0>, rot=<0 0 0 1>}
 bottomt_L_Corner_Up {type=Curve, loc=<0 0 0>, rot=<0 0 0 1>}
 TrailCurve1 {type=Curve, loc=<0 0 0>, rot=<0 0 0 1>}
 BothHandFK {type=Curve, loc=<0 0 0>, rot=<0 0 0 1>}
 bottom_R_Corner_Down1 {type=Curve, loc=<0 0 0>, rot=<0 0 0 1>}
 blendPoint1 {type=Curve, loc=<0 0 0>, rot=<0 0 0 1>}
 WeaponAimingFreeze {type=Curve, loc=<0 0 0>, rot=<0 0 0 1>}
 l_calf_top_taper {type=Curve, loc=<0 0 0>, rot=<0 0 0 1>}
 JogBlendSpaceRotation {type=Curve, loc=<0 0 0>, rot=<0 0 0 1>}
 ZeroOutPitchWeight {type=Curve, loc=<0 0 0>, rot=<0 0 0 1>}
 MeleeTwist {type=Curve, loc=<0 0 0>, rot=<0 0 0 1>}
 MeleeAnim {type=Curve, loc=<0 0 0>, rot=<0 0 0 1>}
 l_sneer_fix {type=Curve, loc=<0 0 0>, rot=<0 0 0 1>}
 jaw_fix {type=Curve, loc=<0 0 0>, rot=<0 0 0 1>}
 r_sneer_fix {type=Curve, loc=<0 0 0>, rot=<0 0 0 1>}
 TurnRotationAmount {type=Curve, loc=<0 0 0>, rot=<0 0 0 1>}
 TurnEnd {type=Curve, loc=<0 0 0>, rot=<0 0 0 1>}
 BlendIn {type=Curve, loc=<0 0 0>, rot=<0 0 0 1>}
 BlendOut {type=Curve, loc=<0 0 0>, rot=<0 0 0 1>}
 BodyRotateX {type=Curve, loc=<0 0 0>, rot=<0 0 0 1>}
 BodyTranslateX {type=Curve, loc=<0 0 0>, rot=<0 0 0 1>}
 CrescentKickScale {type=Curve, loc=<0 0 0>, rot=<0 0 0 1>}
 mask_geo {type=Curve, loc=<0 0 0>, rot=<0 0 0 1>}
 RemoveJumpAdditive {type=Curve, loc=<0 0 0>, rot=<0 0 0 1>}
 RightLeg_IK {type=Curve, loc=<0 0 0>, rot=<0 0 0 1>}
 SnapWeapon_LHand {type=Curve, loc=<0 0 0>, rot=<0 0 0 1>}
 PivotAnim {type=Curve, loc=<0 0 0>, rot=<0 0 0 1>}
 DefaultAdditiveStateWeight {type=Curve, loc=<0 0 0>, rot=<0 0 0 1>}
 StartAnim {type=Curve, loc=<0 0 0>, rot=<0 0 0 1>}
 StopAnim {type=Curve, loc=<0 0 0>, rot=<0 0 0 1>}
 DisableSpeedWarping {type=Curve, loc=<0 0 0>, rot=<0 0 0 1>}
 DisableDBNOSlopeWarp {type=Curve, loc=<0 0 0>, rot=<0 0 0 1>}
 MaskRightArm {type=Curve, loc=<0 0 0>, rot=<0 0 0 1>}
 TurnInPlaceAnim {type=Curve, loc=<0 0 0>, rot=<0 0 0 1>}
 Head_MESH {type=Curve, loc=<0 0 0>, rot=<0 0 0 1>}
 Hair_MESH {type=Curve, loc=<0 0 0>, rot=<0 0 0 1>}
 LaunchPadTransitionOff {type=Curve, loc=<0 0 0>, rot=<0 0 0 1>}
 C_glabella_up_pose {type=Curve, loc=<0 0 0>, rot=<0 0 0 1>}
 jaw_open_pose {type=Curve, loc=<0 0 0>, rot=<0 0 0 1>}
 L_blink_pose {type=Curve, loc=<0 0 0>, rot=<0 0 0 1>}
 L_brow_down_pose {type=Curve, loc=<0 0 0>, rot=<0 0 0 1>}
 L_brow_squeeze_in_pose {type=Curve, loc=<0 0 0>, rot=<0 0 0 1>}
 L_brow_up_pose {type=Curve, loc=<0 0 0>, rot=<0 0 0 1>}
 L_frown_pose {type=Curve, loc=<0 0 0>, rot=<0 0 0 1>}
 L_lip_corner_narrow_pose {type=Curve, loc=<0 0 0>, rot=<0 0 0 1>}
 L_squint_inner_pose {type=Curve, loc=<0 0 0>, rot=<0 0 0 1>}
 L_wide_pose {type=Curve, loc=<0 0 0>, rot=<0 0 0 1>}
 look_down_pose {type=Curve, loc=<0 0 0>, rot=<0 0 0 1>}
 look_left_pose {type=Curve, loc=<0 0 0>, rot=<0 0 0 1>}
 look_right_pose {type=Curve, loc=<0 0 0>, rot=<0 0 0 1>}
 look_up_pose {type=Curve, loc=<0 0 0>, rot=<0 0 0 1>}
 phoneme_fv_pose {type=Curve, loc=<0 0 0>, rot=<0 0 0 1>}
 phoneme_oo_pose {type=Curve, loc=<0 0 0>, rot=<0 0 0 1>}
 R_blink_pose {type=Curve, loc=<0 0 0>, rot=<0 0 0 1>}
 R_brow_down_pose {type=Curve, loc=<0 0 0>, rot=<0 0 0 1>}
 R_brow_squeeze_in_pose {type=Curve, loc=<0 0 0>, rot=<0 0 0 1>}
 R_brow_up_pose {type=Curve, loc=<0 0 0>, rot=<0 0 0 1>}
 R_frown_pose {type=Curve, loc=<0 0 0>, rot=<0 0 0 1>}
 R_lip_corner_narrow_pose {type=Curve, loc=<0 0 0>, rot=<0 0 0 1>}
 R_squint_inner_pose {type=Curve, loc=<0 0 0>, rot=<0 0 0 1>}
 R_wide_pose {type=Curve, loc=<0 0 0>, rot=<0 0 0 1>}
 DisableArmsHeadAdditive {type=Curve, loc=<0 0 0>, rot=<0 0 0 1>}
 FrontendRotation {type=Curve, loc=<0 0 0>, rot=<0 0 0 1>}
 BothLegFK {type=Curve, loc=<0 0 0>, rot=<0 0 0 1>}
 AllowTurnInPlace {type=Curve, loc=<0 0 0>, rot=<0 0 0 1>}
 MontageTimeStretchCurve {type=Curve, loc=<0 0 0>, rot=<0 0 0 1>}
 ParaGlideLeanAlpha {type=Curve, loc=<0 0 0>, rot=<0 0 0 1>}
 WeaponRInChestSpace {type=Curve, loc=<0 0 0>, rot=<0 0 0 1>}
 JimBeam_AO {type=Curve, loc=<0 0 0>, rot=<0 0 0 1>}
 R_squeeze_pose {type=Curve, loc=<0 0 0>, rot=<0 0 0 1>}
 L_squeeze_pose {type=Curve, loc=<0 0 0>, rot=<0 0 0 1>}
 R_upper_lip_raiser_pose {type=Curve, loc=<0 0 0>, rot=<0 0 0 1>}
 L_cheek_up_pose {type=Curve, loc=<0 0 0>, rot=<0 0 0 1>}
 R_cheek_up_pose {type=Curve, loc=<0 0 0>, rot=<0 0 0 1>}
 DisableFaceOverride {type=Curve, loc=<0 0 0>, rot=<0 0 0 1>}
 EnableFullBodyOverride {type=Curve, loc=<0 0 0>, rot=<0 0 0 1>}
 C_glabella_down_pose {type=Curve, loc=<0 0 0>, rot=<0 0 0 1>}
 C_nose_wrinkler_pose {type=Curve, loc=<0 0 0>, rot=<0 0 0 1>}
 eye_convergence_pose {type=Curve, loc=<0 0 0>, rot=<0 0 0 1>}
 jaw_back_pose {type=Curve, loc=<0 0 0>, rot=<0 0 0 1>}
 jaw_forward_pose {type=Curve, loc=<0 0 0>, rot=<0 0 0 1>}
 jaw_left_pose {type=Curve, loc=<0 0 0>, rot=<0 0 0 1>}
 jaw_right_pose {type=Curve, loc=<0 0 0>, rot=<0 0 0 1>}
 jaw_up_pose {type=Curve, loc=<0 0 0>, rot=<0 0 0 1>}
 L_lip_corner_wide_pose {type=Curve, loc=<0 0 0>, rot=<0 0 0 1>}
 L_lower_lip_down_pose {type=Curve, loc=<0 0 0>, rot=<0 0 0 1>}
 L_lower_lip_up_pose {type=Curve, loc=<0 0 0>, rot=<0 0 0 1>}
 L_smile_pose {type=Curve, loc=<0 0 0>, rot=<0 0 0 1>}
 L_upper_lip_lower_pose {type=Curve, loc=<0 0 0>, rot=<0 0 0 1>}
 L_upper_lip_raiser_pose {type=Curve, loc=<0 0 0>, rot=<0 0 0 1>}
 L_upperLidDown_pose {type=Curve, loc=<0 0 0>, rot=<0 0 0 1>}
 mouth_down_pose {type=Curve, loc=<0 0 0>, rot=<0 0 0 1>}
 mouth_up_pose {type=Curve, loc=<0 0 0>, rot=<0 0 0 1>}
 phoneme_ch_pose {type=Curve, loc=<0 0 0>, rot=<0 0 0 1>}
 phoneme_mbp_pose {type=Curve, loc=<0 0 0>, rot=<0 0 0 1>}
 R_lip_corner_wide_pose {type=Curve, loc=<0 0 0>, rot=<0 0 0 1>}
 R_lower_lip_down_pose {type=Curve, loc=<0 0 0>, rot=<0 0 0 1>}
 R_lower_lip_up_pose {type=Curve, loc=<0 0 0>, rot=<0 0 0 1>}
 R_smile_pose {type=Curve, loc=<0 0 0>, rot=<0 0 0 1>}
 R_upper_lip_lower_pose {type=Curve, loc=<0 0 0>, rot=<0 0 0 1>}
 R_upperLidDown_pose {type=Curve, loc=<0 0 0>, rot=<0 0 0 1>}
 tongue_back_pose {type=Curve, loc=<0 0 0>, rot=<0 0 0 1>}
 tongue_down_pose {type=Curve, loc=<0 0 0>, rot=<0 0 0 1>}
 tongue_forward_pose {type=Curve, loc=<0 0 0>, rot=<0 0 0 1>}
 tongue_up_pose {type=Curve, loc=<0 0 0>, rot=<0 0 0 1>}
 LiveLink {type=Curve, loc=<0 0 0>, rot=<0 0 0 1>}
 ParaGlideDeploy {type=Curve, loc=<0 0 0>, rot=<0 0 0 1>}
 HideLeftWeapon {type=Curve, loc=<0 0 0>, rot=<0 0 0 1>}
 DisableLeftHandIKGolfCart {type=Curve, loc=<0 0 0>, rot=<0 0 0 1>}
 DisableRightHandIKGolfCart {type=Curve, loc=<0 0 0>, rot=<0 0 0 1>}
 DisableHandIKRetargeting {type=Curve, loc=<0 0 0>, rot=<0 0 0 1>}
 DisableConsumableOffsetPose {type=Curve, loc=<0 0 0>, rot=<0 0 0 1>}
 UpperBodyNoAdditive {type=Curve, loc=<0 0 0>, rot=<0 0 0 1>}
 MaskLeftArm {type=Curve, loc=<0 0 0>, rot=<0 0 0 1>}
 ForceUpperBody {type=Curve, loc=<0 0 0>, rot=<0 0 0 1>}
 RightHandFK {type=Curve, loc=<0 0 0>, rot=<0 0 0 1>}
 ParaGlideLeanTransition {type=Curve, loc=<0 0 0>, rot=<0 0 0 1>}
 InMotionEmote {type=Curve, loc=<0 0 0>, rot=<0 0 0 1>}
 ForceFullBodyInFrontEnd {type=Curve, loc=<0 0 0>, rot=<0 0 0 1>}
 WeaponLinFloorSpace {type=Curve, loc=<0 0 0>, rot=<0 0 0 1>}
 FrontEnd_Idle_Bypass {type=Curve, loc=<0 0 0>, rot=<0 0 0 1>}
 DisableSlopeWarping {type=Curve, loc=<0 0 0>, rot=<0 0 0 1>}
 Disable_Custom {type=Curve, loc=<0 0 0>, rot=<0 0 0 1>}
 EmoteRequiresWeapon {type=Curve, loc=<0 0 0>, rot=<0 0 0 1>}
 DisableIKRootAdditive {type=Curve, loc=<0 0 0>, rot=<0 0 0 1>}
 DisableFullBodySprint {type=Curve, loc=<0 0 0>, rot=<0 0 0 1>}
 DetachFeet_Vehicle {type=Curve, loc=<0 0 0>, rot=<0 0 0 1>}
 IndigoBoost {type=Curve, loc=<0 0 0>, rot=<0 0 0 1>}
 DisableUpperBodyAdditiveMeshSpace {type=Curve, loc=<0 0 0>, rot=<0 0 0 1>}
 MeleeAOWeight {type=Curve, loc=<0 0 0>, rot=<0 0 0 1>}
 HandsInRootSpace {type=Curve, loc=<0 0 0>, rot=<0 0 0 1>}
 CrouchPelvisAdjustment {type=Curve, loc=<0 0 0>, rot=<0 0 0 1>}
 EnableAccelYawCorrection {type=Curve, loc=<0 0 0>, rot=<0 0 0 1>}
 ParaGlideModifyRoot {type=Curve, loc=<0 0 0>, rot=<0 0 0 1>}
 UseScalarParameterForBlink {type=Curve, loc=<0 0 0>, rot=<0 0 0 1>}
 BlinkControl {type=Curve, loc=<0 0 0>, rot=<0 0 0 1>}
 ForceTargeting {type=Curve, loc=<0 0 0>, rot=<0 0 0 1>}
 BlinkControlProp {type=Curve, loc=<0 0 0>, rot=<0 0 0 1>}
 UseScalarParameterForBlinkProp {type=Curve, loc=<0 0 0>, rot=<0 0 0 1>}
 CrackshotCozyHeadCorrection {type=Curve, loc=<0 0 0>, rot=<0 0 0 1>}
 ForwardMotion {type=Curve, loc=<0 0 0>, rot=<0 0 0 1>}
 EmoteForcedRootYaw {type=Curve, loc=<0 0 0>, rot=<0 0 0 1>}
 MeleeTrailWidth {type=Curve, loc=<0 0 0>, rot=<0 0 0 1>}
 space_world {type=Curve, loc=<0 0 0>, rot=<0 0 0 1>}
 PeelyBlenderFace {type=Curve, loc=<0 0 0>, rot=<0 0 0 1>}
 Width1 {type=Curve, loc=<0 0 0>, rot=<0 0 0 1>}
 OverrideIKHands {type=Curve, loc=<0 0 0>, rot=<0 0 0 1>}
 ParaGlideAddAlpha {type=Curve, loc=<0 0 0>, rot=<0 0 0 1>}
 ParaGlideTurn {type=Curve, loc=<0 0 0>, rot=<0 0 0 1>}
 DisablePelvisAdjustment {type=Curve, loc=<0 0 0>, rot=<0 0 0 1>}
 UncleBrollyDeploy {type=Curve, loc=<0 0 0>, rot=<0 0 0 1>}
 ForceRootSlopeOrientation {type=Curve, loc=<0 0 0>, rot=<0 0 0 1>}
 DisableAdditiveOffsetPose {type=Curve, loc=<0 0 0>, rot=<0 0 0 1>}
 DisableIdleNoiseAdditive {type=Curve, loc=<0 0 0>, rot=<0 0 0 1>}
 UsingConsumable {type=Curve, loc=<0 0 0>, rot=<0 0 0 1>}
 DisableRightArmAdditive {type=Curve, loc=<0 0 0>, rot=<0 0 0 1>}
]
variables [Upperarm_Twist {type=double, default=0}]
