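# Physics asset for the face accessory mesh of the large male Buff Llama player character, setting up physics for its ears, flaps and mouth.
# Capsule bodies cover each Dyn_ bone and the head, with two ears and fourteen flaps constrained directly to the head by narrow angular limits.
# The mouth forms a chain: a master body rotates freely on the head and carries a tightly limited mouth body with three child flaps.
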PhysicsAsset /BRCosmetics/Characters/Player/Male/Large/Bodies/M_LRG_Buff_Llama/Meshes/Parts/M_LRG_Buff_Llama_FaceAcc_Physics
bodies {
 Dyn_Ears [{shapeType=Capsule, length=5, radius=3, shapeName=Dyn_Ears_capsule, center=<1 0 0>, rotation=<rot 90 0 0>}]
 Dyn_Ears1 [{shapeType=Capsule, length=5, radius=3, shapeName=Dyn_Ears1_capsule, center=<1 0 0>, rotation=<rot 90 0 0>}]
 Dyn_Flap1 [{shapeType=Capsule, length=1, radius=1, shapeName=Dyn_Flap1_capsule, center=<1 0 0>, rotation=<rot 90 0 0>}]
 Dyn_Flap10 [{shapeType=Capsule, length=1, radius=1, shapeName=Dyn_Flap10_capsule, center=<1 0 0>, rotation=<rot 90 0 0>}]
 Dyn_Flap11 [{shapeType=Capsule, length=1, radius=1, shapeName=Dyn_Flap11_capsule, center=<1 0 0>, rotation=<rot 90 0 0>}]
 Dyn_Flap12 [{shapeType=Capsule, length=1, radius=1, shapeName=Dyn_Flap12_capsule, center=<1 0 0>, rotation=<rot 90 0 0>}]
 Dyn_Flap13 [{shapeType=Capsule, length=1, radius=1, shapeName=Dyn_Flap13_capsule, center=<1 0 0>, rotation=<rot 90 0 0>}]
 Dyn_Flap14 [{shapeType=Capsule, length=1, radius=1, shapeName=Dyn_Flap14_capsule, center=<1 0 0>, rotation=<rot 90 0 0>}]
 Dyn_Flap2 [{shapeType=Capsule, length=1, radius=1, shapeName=Dyn_Flap2_capsule, center=<1 -0.1124 0>, rotation=<rot 90 0 0>}]
 Dyn_Flap3 [{shapeType=Capsule, length=1, radius=1, shapeName=Dyn_Flap3_capsule, center=<1 0 0>, rotation=<rot 90 0 0>}]
 Dyn_Flap4 [{shapeType=Capsule, length=1, radius=1, shapeName=Dyn_Flap4_capsule, center=<0.9999 -0.8925 0.0001>, rotation=<rot 90 0 0>}]
 Dyn_Flap5 [{shapeType=Capsule, length=1, radius=1, shapeName=Dyn_Flap5_capsule, center=<1 -0.7496 0>, rotation=<rot 90 0 0>}]
 Dyn_Flap6 [{shapeType=Capsule, length=1, radius=1, shapeName=Dyn_Flap6_capsule, center=<1 0 0>, rotation=<rot 90 0 0>}]
 Dyn_Flap7 [{shapeType=Capsule, length=1, radius=1, shapeName=Dyn_Flap7_capsule, center=<1 0 0>, rotation=<rot 90 0 0>}]
 Dyn_Flap8 [{shapeType=Capsule, length=1, radius=1, shapeName=Dyn_Flap8_capsule, center=<1 0 0>, rotation=<rot 90 0 0>}]
 Dyn_Flap9 [{shapeType=Capsule, length=1, radius=1, shapeName=Dyn_Flap9_capsule, center=<1 0 0>, rotation=<rot 90 0 0>}]
 Dyn_Master_Mouth [{shapeType=Capsule, length=0.5, radius=0.5, shapeName=Dyn_Master_Mouth_capsule}]
 Dyn_Mouth [{shapeType=Capsule, length=2, radius=2, shapeName=Dyn_Mouth_capsule, center=<1 0 0>, rotation=<rot 90 0 0>}]
 Dyn_Mouth_Flap1 [{shapeType=Capsule, length=1, radius=1, shapeName=Dyn_Mouth_Flap1_capsule, center=<1 0 0>, rotation=<rot 90 0 0>}]
 Dyn_Mouth_Flap2 [{shapeType=Capsule, length=1, radius=1, shapeName=Dyn_Mouth_Flap2_capsule, center=<1 0 0>, rotation=<rot 90 0 0>}]
 Dyn_Mouth_Flap3 [{shapeType=Capsule, length=1, radius=1, shapeName=Dyn_Mouth_Flap3_capsule, center=<1 0 0>, rotation=<rot 90 0 0>}]
 head [{shapeType=Capsule, length=8.5, radius=9, shapeName=head_capsule, center=<-3.5999 -5.0178 0>, rotation=<rot 90 -7.125 -27.1277>}]
}
constraints [
 {
  bone1Name=Dyn_Master_Mouth
  bone2Name=head
  swing1LimitDegrees=45
  swing1Motion=Free
  swing2LimitDegrees=45
  swing2Motion=Free
  twistLimitDegrees=45
  twistMotion=Free
 }
 {bone1Name=Dyn_Mouth, bone2Name=Dyn_Master_Mouth, swing1LimitDegrees=20, swing2LimitDegrees=1, twistLimitDegrees=1}
 {bone1Name=Dyn_Mouth_Flap1, bone2Name=Dyn_Mouth, swing1LimitDegrees=20, swing2LimitDegrees=1, twistLimitDegrees=1}
 {bone1Name=Dyn_Mouth_Flap2, bone2Name=Dyn_Mouth, swing1LimitDegrees=20, swing2LimitDegrees=1, twistLimitDegrees=1}
 {bone1Name=Dyn_Mouth_Flap3, bone2Name=Dyn_Mouth, swing1LimitDegrees=20, swing2LimitDegrees=1, twistLimitDegrees=1}
 {bone1Name=Dyn_Flap1, bone2Name=head, swing1LimitDegrees=20, swing2LimitDegrees=10, twistLimitDegrees=1}
 {bone1Name=Dyn_Flap2, bone2Name=head, swing1LimitDegrees=20, swing2LimitDegrees=10, twistLimitDegrees=1}
 {bone1Name=Dyn_Flap3, bone2Name=head, swing1LimitDegrees=20, swing2LimitDegrees=10, twistLimitDegrees=1}
 {bone1Name=Dyn_Flap4, bone2Name=head, swing1LimitDegrees=20, swing2LimitDegrees=10, twistLimitDegrees=1}
 {bone1Name=Dyn_Flap5, bone2Name=head, swing1LimitDegrees=20, swing2LimitDegrees=10, twistLimitDegrees=1}
 {bone1Name=Dyn_Flap6, bone2Name=head, swing1LimitDegrees=20, swing2LimitDegrees=10, twistLimitDegrees=1}
 {bone1Name=Dyn_Flap7, bone2Name=head, swing1LimitDegrees=20, swing2LimitDegrees=10, twistLimitDegrees=1}
 {bone1Name=Dyn_Flap8, bone2Name=head, swing1LimitDegrees=20, swing2LimitDegrees=10, twistLimitDegrees=1}
 {bone1Name=Dyn_Flap9, bone2Name=head, swing1LimitDegrees=20, swing2LimitDegrees=10, twistLimitDegrees=1}
 {bone1Name=Dyn_Flap10, bone2Name=head, swing1LimitDegrees=20, swing2LimitDegrees=10, twistLimitDegrees=1}
 {bone1Name=Dyn_Flap11, bone2Name=head, swing1LimitDegrees=20, swing2LimitDegrees=10, twistLimitDegrees=1}
 {bone1Name=Dyn_Flap12, bone2Name=head, swing1LimitDegrees=20, swing2LimitDegrees=10, twistLimitDegrees=1}
 {bone1Name=Dyn_Flap13, bone2Name=head, swing1LimitDegrees=20, swing2LimitDegrees=10, twistLimitDegrees=1}
 {bone1Name=Dyn_Flap14, bone2Name=head, swing1LimitDegrees=20, swing2LimitDegrees=10, twistLimitDegrees=1}
 {bone1Name=Dyn_Ears, bone2Name=head, swing1LimitDegrees=22, swing2LimitDegrees=7, twistLimitDegrees=1}
 {bone1Name=Dyn_Ears1, bone2Name=head, swing1LimitDegrees=22, swing2LimitDegrees=7, twistLimitDegrees=1}
]
